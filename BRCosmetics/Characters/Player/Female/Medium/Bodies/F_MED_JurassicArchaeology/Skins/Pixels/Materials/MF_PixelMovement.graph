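# Material function for the Pixels character skin that animates purple, yellow, blue and pink patterns over incoming material attributes.
# Looping time, remapped through a curve atlas, gives each color its own timing channel.
# SphereMask compares these times against gradient-texture channels in offset, scaled UVs, creating moving bands that Step trims.
# The bands blend colors into base color and combine into a HeartMask; roughness is also lerped toward a lower value.
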MaterialFunction /BRCosmetics/Characters/Player/Female/Medium/Bodies/F_MED_JurassicArchaeology/Skins/Pixels/Materials/MF_PixelMovement
graph {nodes [
  n1 = GetMaterialAttributes(None=FunctionInput() {inputName="Material(MA)", inputType=FunctionInput_MaterialAttributes}) {attributeGetTypes [#1, #2, #3]}
  n2 = Multiply(A=@n1.Roughness) {constB=0.4}
  n3 = NamedRerouteUsage()
  FunctionOutput(None=SetMaterialAttributes!(
    MaterialAttributes=@n1.MaterialAttributes
    BaseColor=NamedRerouteUsage()
    Roughness=LinearInterpolate(A=@n1.Roughness, B=@n2, Alpha=@n3) {constB=0.2}
    Normal=@n1.MaterialAttributes
   ) {attributeSetTypes [#1, #2, #3]})
  n4 = Max(A=NamedRerouteUsage(), B=NamedRerouteUsage())
  n5 = NamedRerouteUsage()
  NamedRerouteDeclaration(None=Max(A=Max(A=@n4, B=@n5), B=NamedRerouteUsage())) {name=HeartMask, nodeColor=<rgba 1 0.0469 0 1>}
  NamedRerouteDeclaration(None=Multiply(A=Time(), B=ScalarParameter() {defaultValue=0.1, group=PixelMovement, parameterName=MovingSpeed}) {constB=4}) {name=TimeDriver, nodeColor=<rgba 0.0781 0 1 1>}
  n6 = VectorParameter() {group=PixelMovement, parameterName=Color_Pink, defaultValue=<rgba 0.9583 0.2733 0.8073 1>}
  n7 = ComponentMask(None=NamedRerouteUsage()) {b=true}
  n8 = Multiply(A=SphereMask(A=@n7, B=NamedRerouteUsage()) {attenuationRadius=0.15}, B=Step(X=@n7) {constY=0.03})
  NamedRerouteDeclaration(None=LinearInterpolate(
    A=NamedRerouteUsage()
    B=NamedRerouteDeclaration(None=@n6.RGB) {name=PinkColor, nodeColor=<rgba 1 0.7266 0 1>}
    Alpha=NamedRerouteDeclaration(None=@n8) {name=DynamicPink, nodeColor=<rgba 0.7109 0 1 1>}
   )) {name=Pattern4ColorOUT, nodeColor=<rgba 0 1 0.8359 1>}
  MaterialExpressionTextureCoordinate_2 = TextureCoordinate()
  MaterialExpressionAdd_3 = Add(
   A=AppendVector(
    A=ScalarParameter() {group=PixelMovement, parameterName=Shape_UVOffsetX01, sortPriority=62}
    B=ScalarParameter() {group=PixelMovement, parameterName=Shape_UVOffsetY01, sortPriority=63}
   )
   B=Constant2Vector() {g=0.5, r=0.5}
  )
  MaterialExpressionScalarParameter_47 = ScalarParameter() {defaultValue=1, group=PixelMovement, parameterName=Shape_UVScale01, sortPriority=65}
  n9 = Custom!(OriginalUV=@MaterialExpressionTextureCoordinate_2, UVPoint=@MaterialExpressionAdd_3, UVScale=@MaterialExpressionScalarParameter_47) {
   code="float2 DistUV =OriginalUV-UVPoint;\r\nDistUV = DistUV/(UVScale/2) + float2(0.5,0.5);\r\n\r\nreturn DistUV;"
   description=CopyToPointUV
   outputType=CMOT_Float2
   inputs [
    {
     inputName=OriginalUV
     input {expression=&self:MaterialExpressionTextureCoordinate_2}
    }
    {
     inputName=UVPoint
     input {expression=&self:MaterialExpressionAdd_3}
    }
    {
     inputName=UVScale
     input {expression=&self:MaterialExpressionScalarParameter_47}
    }
   ]
  }
  NamedRerouteDeclaration(None=@n9) {name=PatternUV, nodeColor=<rgba 0.5703 0 1 1>}
  n10 = CurveAtlasRowParameter(CurveTime=Frac(None=NamedRerouteUsage())) {
   atlas=&/A_PixelTimeRemap
   curve=&/TimeFracRemapPixel
   group=PixelMovement
   parameterName=TimeRemap
  }
  NamedRerouteDeclaration(None=@n10.R) {name=PurpleTime, nodeColor=<rgba 0 1 0.3438 1>}
  n11 = NamedRerouteUsage()
  n12 = TextureSampleParameter2D(UVs=@n11) {
   group=PixelMovement
   parameterName=MovenmentPattern0
   samplerType=SAMPLERTYPE_Masks
   texture=/BRCosmetics/Characters/Player/Female/Medium/Bodies/F_MED_JurassicArchaeology/Skins/Pixels/Textures/T_PYB_Grad
  }
  n13 = Multiply(
   A=Saturate(None=MaterialFunctionCall!(UV_Coord=@n11) {materialFunction=&/MF_UV01_Mask}.Mask)
   B=@n12.RGB
  )
  n14 = ComponentMask(None=@n13) {r=true}
  n15 = NamedRerouteUsage()
  NamedRerouteDeclaration(None=Multiply(A=SphereMask(A=@n14, B=@n15) {attenuationRadius=0.15}, B=Step(X=@n14) {constY=0.03})) {name=DynamicPurple, nodeColor=<rgba 0.3606 0 1 1>}
  n16 = ComponentMask(None=@n13) {g=true}
  n17 = NamedRerouteUsage()
  NamedRerouteDeclaration(None=Multiply(A=SphereMask(A=@n16, B=@n17) {attenuationRadius=0.15}, B=Step(X=@n16) {constY=0.03})) {name=DynamicYellow, nodeColor=<rgba 1 0.773 0 1>}
  NamedRerouteDeclaration(None=@n10.G) {name=YellowTime, nodeColor=<rgba 0 1 0.1328 1>}
  NamedRerouteDeclaration(None=@n10.B) {name=BlueTime, nodeColor=<rgba 0.5234 0 1 1>}
  NamedRerouteDeclaration(None=@n10.A) {name=PinkTime, nodeColor=<rgba 1 0.6094 0 1>}
  NamedRerouteDeclaration(None=@n1.BaseColor) {name=BaseColorIN, nodeColor=<rgba 1 0 0.5156 1>}
  n18 = ComponentMask(None=@n13) {b=true}
  n19 = NamedRerouteUsage()
  NamedRerouteDeclaration(None=Multiply(A=SphereMask(A=@n18, B=@n19) {attenuationRadius=0.15}, B=Step(X=@n18) {constY=0.03})) {name=DynamicBlue, nodeColor=<rgba 0 0.0857 1 1>}
  n20 = VectorParameter() {group=PixelMovement, parameterName=Color_Yellow, defaultValue=<rgba 0.62 0.4968 0 1>}
  n21 = LinearInterpolate(
   A=LinearInterpolate(A=NamedRerouteUsage(), B=NamedRerouteDeclaration(None=@n20.RGB) {name=YellowColor, nodeColor=<rgba 0.9922 0 1 1>}, Alpha=NamedRerouteUsage())
   B=NamedRerouteDeclaration(None=VectorParameter() {group=PixelMovement, parameterName=Color_Blue, defaultValue=<rgba 0.0035 0 0.51 1>}.RGB) {name=BlueColor, nodeColor=<rgba 0 0.25 1 1>}
   Alpha=NamedRerouteUsage()
  )
  n22 = NamedRerouteDeclaration(None=VectorParameter() {group=PixelMovement, parameterName=Color_Purple, defaultValue=<rgba 0.1455 0 0.25 1>}.RGB) {name=PurpleColor, nodeColor=<rgba 0.0312 1 0 1>}
  n23 = NamedRerouteUsage()
  NamedRerouteDeclaration(None=NamedRerouteDeclaration(None=LinearInterpolate(A=@n21, B=@n22, Alpha=@n23)) {name=3ColorBlend, nodeColor=<rgba 1 0.0469 0 1>}) {name=Pattern3ColorOUT, nodeColor=<rgba 0.6406 0 1 1>}
  n24 = TextureSampleParameter2D(UVs=Clamp(None=NamedRerouteUsage())) {
   group=PixelMovement
   parameterName=MovenmentPatternPink
   samplerType=SAMPLERTYPE_LinearColor
   texture=/BRCosmetics/Characters/Player/Female/Medium/Bodies/F_MED_JurassicArchaeology/Skins/Pixels/Textures/T_PinkAlpha_Grad
  }
  NamedRerouteDeclaration(None=@n24.RGB) {name=PinkGrad, nodeColor=<rgba 0.9219 1 0 1>}
 ]}
